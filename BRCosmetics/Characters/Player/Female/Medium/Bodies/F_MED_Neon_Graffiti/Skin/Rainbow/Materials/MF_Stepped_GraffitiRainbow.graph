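# Material function for the Neon Graffiti Rainbow skin that rewrites BaseColor and EmissiveColor with animated rainbow hues.
# Time cycles a masked hue shift on BaseColor and a generated rainbow, quantized into columns by ceiling scaled UV1.
# "Is Body" switches the rainbow mask from a texture channel to one derived from input emissive.
# Static switches toggle the generated rainbow, masked panning emissive pulses, and rainbow glow with time-of-day night boost.
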MaterialFunction /BRCosmetics/Characters/Player/Female/Medium/Bodies/F_MED_Neon_Graffiti/Skin/Rainbow/Materials/MF_Stepped_GraffitiRainbow
graph {nodes [
  n1 = GetMaterialAttributes(None=FunctionInput() {inputType=FunctionInput_MaterialAttributes}) {attributeGetTypes [#1, #2]}
  n2 = TextureSampleParameter2D() {
   group=NG_Rainbow
   parameterName=NG_Rainbow_HueShiftMask
   samplerSource=SSM_Wrap_WorldGroupSettings
   texture=/BRCosmetics/Characters/Player/Female/Medium/Bodies/F_MED_Neon_Graffiti/Skin/Rainbow/Textures/T_F_MED_Neon_Graffiti_Rainbow_FaceAcc_FX
  }
  n3 = StaticComponentMaskParameter(None=@n2.RGB) {defaultB=true, group=NG_Rainbow, parameterName=NG_Rainbow_HueShiftMaskChannel}
  n4 = TextureCoordinate()
  n5 = SphereMask(A=ComponentMask(None=@n4) {g=true}, B=Constant() {r=0.8}, Radius=Constant() {r=0.122})
  n6 = SphereMask(A=ComponentMask(None=@n4) {r=true}, B=Constant() {r=0.177}, Radius=Constant() {r=0.092})
  Add(A=Ceil(None=ComponentMask(None=@n1.EmissiveColor) {r=true}), B=Multiply(A=OneMinus(None=@n3), B=Multiply(A=@n5, B=@n6)))
  n7 = Multiply(A=Time(), B=ScalarParameter() {defaultValue=1.5, group=NG_Rainbow, parameterName=NG_Rainbow_Speed}) {constB=1.5}
  Ceil(None=@n7)
  n8 = MaterialFunctionCall(Float3=@n1.EmissiveColor) {materialFunction=&BreakOutFloat3Components}
  n9 = Add(A=Add(A=@n8.R, B=@n8.R), B=@n8.B)
  Ceil(None=@n9)
  n10 = ScalarParameter() {defaultValue=22, group=NG_Rainbow, parameterName=NG_Rainbow_ColumnNumber}
  n11 = Fmod(A=@n7, B=@n10)
  n12 = Divide(B=@n10)
  n13 = LinearInterpolate(
   A=@n1.BaseColor
   B=MaterialFunctionCall(
    Hue Shift Percentage=Multiply(A=@n11, B=@n12)
    Texture=MaterialFunctionCall(Base=@n1.BaseColor, Blend=@n1.BaseColor) {materialFunction=&Blend_Overlay}.Result
   ) {materialFunction=&HueShift}.Result
   Alpha=StaticSwitchParameter(True=Constant(), False=@n3) {group=NG_Rainbow, parameterName=Is Body, sortPriority=0}
  )
  n14 = ScalarParameter() {defaultValue=22, group=NG_Rainbow, parameterName=NG_Rainbow_ColumnNumber}
  n15 = TextureCoordinate() {coordinateIndex=1}
  n16 = Multiply(A=Add(B=ComponentMask(None=@n15) {r=true}), B=@n14)
  n17 = Multiply(
   A=Multiply(A=Ceil(None=@n16), B=Divide(B=@n14)) {constB=0.1}
   B=Divide(A=@n14, B=ScalarParameter() {defaultValue=14, group=NG_Rainbow, parameterName=NG_Rainbow_StepNumber})
  ) {constB=0.1}
  n18 = Constant() {r=0.1}
  n19 = MaterialFunctionCall(
   Hue Shift Percentage=Multiply(A=Fmod(A=@n7, B=@n14), B=Divide(B=@n14))
   Texture=MaterialFunctionCall(Hue Shift Percentage=Add(A=@n17, B=@n18)) {materialFunction=&HueShift}.Result
  ) {materialFunction=&HueShift}
  n20 = StaticSwitchParameter(
   True=Saturate(None=Power(Base=@n9) {constExponent=0.45})
   False=StaticComponentMaskParameter(None=@n2.RGB) {defaultB=true, group=NG_Rainbow, parameterName=NG_Rainbow_RainbowMaskChannel}
  ) {group=NG_Rainbow, parameterName=Is Body, sortPriority=0}
  n21 = LinearInterpolate(A=@n13, B=@n19.Result, Alpha=@n20)
  n22 = StaticSwitchParameter(True=@n21, False=@n13) {group=NG_Rainbow, parameterName=Use Generated Rainbow, sortPriority=0}
  n23 = LinearInterpolate(B=@n19.Result, Alpha=@n20)
  n24 = ScalarParameter() {defaultValue=0.25, group=NG_Rainbow, parameterName=EmissiveMult}
  n25 = Constant() {r=1}
  n26 = MaterialFunctionCall(Value_Day=@n25, Value_Night=Multiply(A=@n25, B=ScalarParameter() {defaultValue=1, group=NG_Rainbow, parameterName=NG_Rainbow_EMSNightBoost})) {materialFunction=&Scalar_Time_of_Day}
  n27 = Panner(Coordinate=@n4, Time=Time(), Speed=ScalarParameter() {defaultValue=-0.35, group=NG_Rainbow, parameterName=NG_Rainbow_EmsSpeed})
  n28 = Frac(None=Multiply(A=ComponentMask(None=@n27) {r=true}) {constB=3})
  n29 = LinearInterpolate(
   A=ScalarParameter() {defaultValue=2, group=NG_Rainbow, parameterName=NG_Rainbow_EmsMin}
   B=ScalarParameter() {defaultValue=25, group=NG_Rainbow, parameterName=NG_Rainbow_EmsMax}
   Alpha=Saturate(None=Add(A=@n28) {constB=-0.6})
  )
  n30 = StaticSwitchParameter(
   True=LinearInterpolate(
    A=@n1.EmissiveColor
    B=Multiply(A=@n22, B=@n29)
    Alpha=StaticComponentMaskParameter(None=@n2.RGB) {defaultG=true, group=NG_Rainbow, parameterName=NG_Rainbow_EmissiveMaskChannel}
   )
   False=@n1.EmissiveColor
  ) {group=NG_Rainbow, parameterName=Use Graffiti Panning Emissive, sortPriority=0}
  n31 = Add(A=Multiply(A=@n23, B=Multiply(A=@n24, B=@n26.Result)) {constB=0.25}, B=@n30)
  n32 = StaticSwitchParameter(True=@n23, False=@n30) {group=NG_Rainbow, parameterName=Is Body, sortPriority=0}
  FunctionOutput(None=SetMaterialAttributes!(
    MaterialAttributes=@n1.MaterialAttributes
    BaseColor=@n22
    EmissiveColor=StaticSwitchParameter(True=@n31, False=@n32) {group=NG_Rainbow, parameterName=Rainbow Emissive Toggle, sortPriority=0}
   ) {attributeSetTypes [#1, #2]})
  Reroute(None=@n21)
  TextureSample() {texture=/BRCosmetics/Characters/Player/Female/Medium/Bodies/F_MED_Neon_Graffiti/Skin/Rainbow/Textures/T_F_MED_Neon_Graffiti_Rainbow_Body_E}
 ]}
