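# Master character material for the Rainbow skin of the female medium Neon Graffiti body cosmetic, built for skeletal meshes with morph targets and mesh deformers.
# It chains shared material functions through material attributes: the base PreFX function feeds CharacterSkinFX, then MF_Stepped_GraffitiRainbow adds the skin-specific effect.
# The base PostFX function finishes the result, also taking outputs from PreFX.
# The shading model comes from the material expression, allowing several models in one material.
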Material /BRCosmetics/Characters/Player/Female/Medium/Bodies/F_MED_Neon_Graffiti/Skin/Rainbow/Materials/MAT_Neon_Graffiti_Master
graph {
 properties {
  bUseMaterialAttributes=true
  bUsedWithMeshDeformer=true
  bUsedWithMorphTargets=true
  bUsedWithSkeletalMesh=true
  pixelDepthOffsetMode=PDOM_Legacy
  refractionMethod=RM_IndexOfRefraction
  shadingModel=MSM_FromMaterialExpression
  usedShadingModels=Default Lit | Subsurface | Clear Coat | Subsurface Profile | Cloth | Eye
 }
 nodes [
  n1 = MaterialFunctionCall() {materialFunction=&BaseFN_CharacterFunction_PreFX}
  n2 = MaterialFunctionCall!(In=MaterialFunctionCall(Gmap=@n1.Gmap, MaterialAttributes=@n1.Gmap, AO=@n1.Gmap) {materialFunction=&CharacterSkinFX}.Result) {materialFunction=&/MF_Stepped_GraffitiRainbow}
  MaterialFunctionCall(AO=@n1.AO, Distance Blend=@n1.AO, Fresnel=@n1.AO, Gmap=@n1.AO, Pre Skinned Local Position=@n1.AO, Material=@n2.Result) {materialFunction=&BaseFN_CharacterFunction_PostFX}
 ]
}
